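# Sound cue for a Season 17 battle pass CB radio voice line, MARI_14Days_2021_08_SoIJust_010.
# A single dialogue player node plays the matching Dialogue Wave with the S17QuestVO speaker and no targets.
# The cue uses the S17 BR quest VO attenuation and routes through the narrator VO sound class.
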SoundCue /BattlepassS17/Sounds/VO/CB_Radio/Cues/08/MARI_14Days_2021_08_SoIJust_010_Dialogue_Cue
root=DialoguePlayer_0
nodes [{
  id=DialoguePlayer_0
  properties {dialogueWaveParameter {
    dialogueWave=/BattlepassS17/Sounds/VO/CB_Radio/Waves/MARI_14Days_2021_08_SoIJust_010_Dialogue
    context {speaker=/BattlepassS17/Sounds/VO/S17QuestVO, targets []}
   }}
 }]
properties {
 attenuationSettings=/BattlepassS17/Sounds/VO/S17_BR_Quest_VO_Att
 maxDistance=1124
 soundClassObject=/Game/Sounds/FortSoundClassesAndModes/Fort_VO_Narrator
 volumeMultiplier=1
}
